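# GetGradientMapRow is a library material function that computes the UV coordinate for sampling one gradient row in a gradient atlas texture.
# It divides the zero-based index by the atlas height and adds half a texel (0.5 over the height) to land on the row's centre.
# The CurveTime input becomes U and that row position becomes V in the appended output.
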MaterialFunction /Engine/Functions/Engine_MaterialFunctions02/Gradients/GetGradientMapRow
graph {
 properties {
  bExposeToLibrary=true
  description=Gets the row needed for a given gradient index in a gradient atlas map of a given size
  libraryCategoriesText [Misc,Gradient]
 }
 nodes [
  n1 = FunctionInput() {
   bUsePreviewValueAsDefault=true
   description="The first index is 0 and the last index is (number of gradients in the gradient map) -1."
   inputName="Index (0 Based)"
   inputType=FunctionInput_Scalar
  }
  n2 = FunctionInput() {inputName=Atlas Height, inputType=FunctionInput_Scalar}
  n3 = Divide(A=@n1, B=@n2)
  n4 = Divide(A=Constant() {r=0.5}, B=@n2)
  FunctionOutput(None=AppendVector(A=FunctionInput() {inputName=CurveTime, inputType=FunctionInput_Scalar}, B=Add(A=@n3, B=@n4)))
 ]
}
