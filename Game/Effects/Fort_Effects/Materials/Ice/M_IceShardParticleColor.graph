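# Masked, two-sided subsurface material for ice shard mesh particles in Niagara.
# An inverted-U gradient blends two ice colors, while the SnowOffset dynamic parameter shapes the same gradient to vary roughness and normal detail.
# A sharpened skylight cubemap sampled by reflection vector fakes reflective emissive, and fire noise times particle alpha forms a mask.
# Dynamic parameters also scale subsurface color and drive a sine-plus-rotation position offset.
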Material /Game/Effects/Fort_Effects/Materials/Ice/M_IceShardParticleColor
graph {
 properties {
  bUsedWithNiagaraMeshParticles=true
  blendMode=BLEND_Masked
  pixelDepthOffsetMode=PDOM_Legacy
  shadingModel=MSM_Subsurface
  translucencyLightingMode=TLM_Surface
  twoSided=true
  usedShadingModels=Subsurface
 }
 nodes [
  n1 = VectorParameter() {defaultValue=<rgba 1 1 20 0>}
  n2 = DynamicParameter() {defaultValue=<rgba 1 1 0 1>, paramNames [SnowOffset,SSS_Mult,WPO_Rotate,Param4]}
  n3 = Add(A=@n2.WPO_Rotate, B=@n1.RGB)
  n4 = Multiply(A=@n1.RGB) {constB=0.2}
  Add(
   A=Multiply(A=Sine(None=@n3) {period=3.3}, B=Constant3Vector() {constant=<rgba 0 0 3 0>})
   B=RotateAboutAxis(
    NormalizedRotationAxis=Constant3Vector() {constant=<rgba 0 0 1 0>}
    RotationAngle=Multiply(A=@n2.WPO_Rotate, B=@n4) {constB=0.1}
    PivotPoint=ActorPositionWS()
    Position=WorldPosition().XYZ
   )
  )
  n5 = OneMinus(None=ComponentMask(None=TextureCoordinate()) {r=true})
  n6 = Power(Base=@n5, Exp=ScalarParameter() {defaultValue=4, parameterName=Exp}) {constExponent=6}
  CurveAtlasRowParameter(CurveTime=@n6) {
   atlas=&AthenaCurveAtlas01
   curve=/Game/Athena/LinearColorCurves/IceFlamesCurve
   defaultValue=0.0215
   parameterName=Curve
  }
  Desaturation(Fraction=Constant() {r=0.5})
  n7 = Clamp(None=Power(Base=@n5, Exp=DynamicParameter() {defaultValue=<rgba 1 1 0 1>, paramNames [SnowOffset,SSS_Mult,WPO_Rotate,Param4]}.SnowOffset))
  LinearInterpolate(
   A=ScalarParameter() {defaultValue=0.85, parameterName=RoughnessHi}
   B=ScalarParameter() {defaultValue=0.1, parameterName=RoughnessLo}
   Alpha=OneMinus(None=@n7)
  )
  MaterialFunctionCall!(Atlas Height, CurveTime, "Index (0 Based)") {materialFunction=/Engine/Functions/Engine_MaterialFunctions02/Gradients/GetGradientMapRow}
  n8 = TextureSample(UVs=ReflectionVectorWS()) {texture=/Game/Packages/Fortress_Sky/TexturesHDR/T_Cubemap_Skylight_Subtle}
  n9 = ScalarParameter() {defaultValue=6, parameterName=Exp_1}
  MaterialFunctionCall(Emissive=Multiply(A=Power(Base=@n8.RGB, Exp=@n9), B=ScalarParameter() {defaultValue=1.5, parameterName=Reflection Boost}) {constB=0.1}) {materialFunction=&MF_FX_EmissiveAdjustments}
  n10 = LinearInterpolate(
   A=VectorParameter() {parameterName=Color2, defaultValue=<rgba 0 0.1854 0.33 1>}.RGB
   B=VectorParameter() {parameterName=Color1, defaultValue=<rgba 0.4619 0.4982 0.5312 0>}.RGB
   Alpha=@n6
  )
  Multiply(A=@n10, B=@n2.SSS_Mult)
  Multiply(A=@n10, B=ScalarParameter() {defaultValue=1.5, parameterName=colour boost})
  Multiply(
   A=Multiply(
    A=TextureSample() {texture=/Game/Global/Textures/Default/Noise/T_FireNoise_01}.RGB
    B=ScalarParameter() {defaultValue=20, parameterName=IceMaskMult}
   )
   B=ParticleColor().A
  )
  n11 = TextureSample(UVs=Multiply(A=TextureCoordinate(), B=ScalarParameter() {defaultValue=1, parameterName=NormalCoors})) {samplerType=SAMPLERTYPE_Normal, texture=/Game/Packages/Fortnite_UniqueMaterials_1/Textures/FoamNormals}
  n12 = Constant() {r=1}
  n13 = VectorParameter() {parameterName=ScaleNormal, defaultValue=<rgba 1 1 0.05 0>}
  n14 = ScalarParameter() {defaultValue=2, parameterName=SSS_Exp}
  Reroute(None=LinearInterpolate(
    A=MaterialFunctionCall(Normal=@n11.RGB, Flatness=@n12) {materialFunction=&FlattenNormal}.Result
    B=Multiply(A=@n11.RGB, B=@n13.RGB)
    Alpha=Power(Base=@n7, Exp=@n14)
   ))
 ]
}
